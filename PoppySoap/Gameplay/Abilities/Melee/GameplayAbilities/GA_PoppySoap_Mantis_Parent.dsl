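; Parent Mantis melee ability for PoppySoap that pushes a camera mode while active and pops it on end.
; Technique events apply an AirKick cooldown to prevent spamming and a knockback effect on tagged uppercut hits, including on the server.
; Without authority, Branch notifies with an Off rule start secondary fire if that input is held.
; Root motion warping is overridden to offset from and face a valid aim-assist target.
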(event Ability|EventActivateAbility
  ;[0.01487 0.15 0.03497 0.5] ActivateAbility: Push the camera mode for the mantis attacks.
  (|CallParentFunction)
  (Camera|PushCameraMode (Variables|Camera|GetMantisCameraModeClass)))

(event Ability|EventOnEndAbility
  (params (bWasCancelled :type Boolean))
  ;[0.15 0.1058 0.00828 0.5] EndAbility: Clean up.
  (Utilities|FlowControl|Sequence
    (:then_0
      (Camera|PopCameraMode))
    (:then_1
      (|CallParentFunction bWasCancelled))))

(event Mantis|EventOnMantisAblilityTechniqueStarted
  (params (OwningTags :type Gameplay%20Tag%20Container%20Structure) (SequenceName :type Name :default "None") (StartsSequence :type Boolean) (EndsSequence :type Boolean))
  ;[0.04346 0.15 0.1471 0.5] OnMantisAbilityTechniqueStarted: AirKick can be spammed, add a cooldown for it otherwise.
  (|CallParentFunction OwningTags SequenceName StartsSequence EndsSequence)
  (if (== SequenceName "AirKick")
    (Ability|ApplyGameplayEffectToOwner "/PoppySoap/Gameplay/Abilities/Melee/GameplayEffects/GE_PoppySoap_Melee_AirKick_Cooldown.GE_PoppySoap_Melee_AirKick_Cooldown_C")))

(event Mantis|EventOnTechniqueNotifyEvent
  (params (NotifyEvent :type EFortMantisNotifyEvent%20Enum :default "Invalid") (Notify :type Mantis%20Event%20Object%20Reference))
  ;[0.07373 0.05888 0.15 0.5] TechniqueNotifyEvent: Invoke secondary input if it is held whenever the notify event of (Branch, Off) occurs.
  (|CallParentFunction NotifyEvent Notify)
  (Ability|HasAuthority
    (:False
      (Utilities|IsValid Notify
        (:Is%20Valid
          (if (and (Utilities|Enum|Equal(Enum) (Variables|/Script/MantisRuntime.FortAnimNotify_Mantis|GetMantisBranchRule Notify) "Off") (Utilities|Enum|Equal(Enum) NotifyEvent "Branch"))
            (bind _outfortplayerpawn (CallFunction|GetAvatarasValidFortPlayerPawn)
              (:ValidFortPlayerPawn
                (if (Variables|/Script/FortniteGame.FortPawn|GetbSecondaryInputHeld _outfortplayerpawn)
                  (Weapon|PawnStartSecondaryFire _outfortplayerpawn 0))))))))))

(event Mantis|EventOnMantisTechniqueHit
  (params (TargetDataHandle :type Gameplay%20Ability%20Target%20Data%20Handle%20Structure%20%28by%20ref%29) (ApplicationTag :type Gameplay%20Tag%20Structure))
  ;[0.15 0.1373 0.02852 0.5] Apply knockback on uppercut technique hit. Done here so it is applied on the server as well instead of just the client
  (|CallParentFunction TargetDataHandle ApplicationTag)
  (if (== ApplicationTag (Variables|Default|GetMeleeKnockbackTechniqueTag))
    (Ability|ApplyGameplayEffectToTarget TargetDataHandle (Variables|Default|GetMeleeKnockbackGE))))

(fn ConfigureRootMotionWarpingAlternativeBehavior
  (params (TechniqueName :type Name :default "None"))
  (bind (_allowtranslationwarp _allowrotationwarp _snapshottargetlocation) (CallFunction|GetRootMotionOptions))
  (bind _aimassisttarget (Variables|Default|GetAimAssistTarget))
  ;[0.07483 0.0513 0.15 0.5] Overridden to set an offset to the target actor while warping.
  (Utilities|FlowControl|Sequence
    (:then_0
      (CallFunction|GetAimAssistTarget)
      (if (Variables|Default|GetHasAimAssistTarget)
        (bind _returnvalue (CallFunction|IsValidTarget _aimassisttarget))
        (if _returnvalue
          (return _aimassisttarget _snapshottargetlocation "OffsetFromTargetActor" (Variables|Mantis|GetWarpTranslationOffsetX) 0.0 0.0 "FaceTargetActor" "0, 0, 0"))))
    (:then_1
      (return 0 false "None" "0, 0, 0" "None" "0, 0, 0"))))

(fn OverrideFailedReason
  (params (FailedReason :type Gameplay%20Tag%20Container%20Structure%20%28by%20ref%29))
  (return 0))
