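; Uppercut knockback ability for Poppy Soap melee hits that overrides DoLaunch to add toasty and announcer laugh effects.
; After calling the parent launch, the authority rolls a hotfixable chance to apply the toasty gameplay effect to the launcher.
; A second hotfix also applies it to the ability owner on a successful roll.
; Separately, if the avatar pawn is still alive, the launcher receives the announcer laugh gameplay effect.
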(event EventDoLaunch
  (params (Vel :type Vector :default "0, 0, 0"))
  (bind _toastygameplayeffectclass (Variables|Default|GetToastyGameplayEffectClass))
  ;[0.02282 0.0429 0.15 0.5] DoLaunch: Chance to execute toasty and announcer laugh VO.
  (|CallParentFunction Vel)
  (Ability|HasAuthority
    (:True
      (Utilities|FlowControl|Sequence
        (:then_0
          (if (>= (ScalableFloat|GetCachedScalableFloatValue (Variables|Default|GetHotfix_ToastyChance)) (Math|Random|RandomFloatinRange 0.001 1.0))
            (Fortnite|FortShippingLog self "DoLaunch: Toasty invoked! ")
            (CallFunction|GetAvatarasValidFortPlayerPawn
              (:ValidFortPlayerPawn
                (bind _asfort_player_pawn (Utilities|Casting|CastToFortPlayerPawn (Variables|KnockbackParent|GetLauncher)))
                (Ability|ApplyGameplayEffectToTarget (Ability|TargetData|AbilityTargetDatafromActor _asfort_player_pawn) _toastygameplayeffectclass 0 0)
                (if (> (ScalableFloat|GetCachedScalableFloatValue (Variables|Default|GetHotfix_ToastyPlayOnKnockbackTarget)) 0)
                  (Ability|ApplyGameplayEffectToOwner _toastygameplayeffectclass 0 0))))))
        (:then_1
          (bind _outfortplayerpawn (CallFunction|GetAvatarasValidFortPlayerPawn)
            (:ValidFortPlayerPawn
              (if (not (FortPawn|IsDead _outfortplayerpawn))
                (bind _asfort_player_pawn_2 (Utilities|Casting|CastToFortPlayerPawn (Variables|KnockbackParent|GetLauncher)))
                (Ability|ApplyGameplayEffectToTarget (Ability|TargetData|AbilityTargetDatafromActor _asfort_player_pawn_2) (Variables|Default|GetAnnouncerLaughGameplayEffect) 0 0)))))))))
